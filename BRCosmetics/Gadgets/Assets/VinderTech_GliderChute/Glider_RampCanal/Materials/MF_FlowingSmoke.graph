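# Material function in the VinderTech RampCanal glider folder that outputs an animated flowing smoke mask.
# A flowmap swirl, scrolling noise and panned base and detail smoke textures provide distortion and layering, combined through max, power and multiply.
# A channel-selectable mask texture is sampled with slightly offset UVs, its contrast pulsing through a time-driven sine exponent.
# Tiling, speed and intensity inputs default to FlowingSmoke parameters, though most input names say FireSmoke.
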MaterialFunction /BRCosmetics/Gadgets/Assets/VinderTech_GliderChute/Glider_RampCanal/Materials/MF_FlowingSmoke
graph {nodes [
  n1 = NamedRerouteUsage()
  n2 = NamedRerouteUsage()
  n3 = Multiply(
   A=NamedRerouteUsage()
   B=FunctionInput(Preview=ScalarParameter() {defaultValue=0.4, group=FlowingSmoke, parameterName=FlowingSmoke_SwirlIntensity}) {bUsePreviewValueAsDefault=true, inputName=FireSmoke_SwirlIntensity, inputType=FunctionInput_Scalar, sortPriority=52}
  ) {constB=2}
  n4 = ScalarParameter() {defaultValue=3, group=FlowingSmoke, parameterName=FlowingSmoke_DetailSmoketiling}
  n5 = VectorParameter() {group=FlowingSmoke, parameterName=FlowingSmoke_DetailSmoke, defaultValue=<rgba -0.04 0.4 0 1>}
  n6 = Panner(
   Coordinate=Multiply(
    A=NamedRerouteUsage()
    B=FunctionInput(Preview=@n4) {bUsePreviewValueAsDefault=true, inputName=FireSmoke_DetailSmoketiling, inputType=FunctionInput_Scalar, sortPriority=55}
   ) {constB=2}
   Speed=ComponentMask(None=FunctionInput(Preview=@n5.RGB) {bUsePreviewValueAsDefault=true, inputName=FireSmoke_DetailSmokeSpeed, inputType=FunctionInput_Vector2, sortPriority=55}) {g=true, r=true}
  ) {speedX=-0.02, speedY=-0.25}
  n7 = TextureSample(UVs=Add(A=Multiply(A=@n2) {constB=0.05}, B=Add(A=@n3, B=@n6))) {texture=/Game/Global/Textures/Default/Noise/T_FireNoise_02}
  n8 = NamedRerouteUsage()
  n9 = Saturate(None=Power(Base=Max(A=@n7.R, B=@n8)) {constExponent=3})
  FunctionOutput(None=NamedRerouteDeclaration(None=Multiply(A=@n1, B=@n9)) {name=DetailSmoke, nodeColor=<rgba 1 0.3281 0 1>}) {outputName=Out_FireSmokeMask}
  n10 = Time()
  n11 = FunctionInput(Preview=ScalarParameter() {defaultValue=0.1, group=FlowingSmoke, parameterName=FlowingSmoke_SwirlUVSpeed}) {bUsePreviewValueAsDefault=true, inputName=FireSmoke_SwirlUVSpeed, inputType=FunctionInput_Scalar, sortPriority=52}
  n12 = NamedRerouteUsage()
  n13 = FunctionInput(Preview=ScalarParameter() {defaultValue=3, group=FlowingSmoke, parameterName=FlowingSmoke_SwirlUVTiling}) {bUsePreviewValueAsDefault=true, inputName=FireSmoke_SwirlUVTiling, inputType=FunctionInput_Scalar, sortPriority=52}
  n14 = TextureSample(UVs=Add(A=Multiply(A=@n10, B=@n11), B=Multiply(A=@n12, B=@n13) {constB=4})) {
   samplerSource=SSM_Wrap_WorldGroupSettings
   samplerType=SAMPLERTYPE_LinearColor
   texture=/Game/VisualThreatManager/StormVisuals/LargeClouds/T_Swirl_Flowmap
  }
  NamedRerouteDeclaration(None=Subtract(A=ComponentMask(None=@n14.RGB) {g=true, r=true}) {constB=0.5}) {name=Swirl UV, nodeColor=<rgba 0 1 0.9297 1>}
  n15 = TextureCoordinate()
  n16 = Multiply(
   A=NamedRerouteUsage()
   B=FunctionInput(Preview=ScalarParameter() {defaultValue=0.02, group=FlowingSmoke, parameterName=FlowingSmoke_NoiseIntensity}) {bUsePreviewValueAsDefault=true, inputName=FireSmoke_NoiseIntensity, inputType=FunctionInput_Scalar, sortPriority=54}
  ) {constB=0.01}
  n17 = ChannelMaskParameter(None=TextureSample(UVs=Add(A=@n15, B=@n16), Tex=NamedRerouteUsage()) {samplerType=SAMPLERTYPE_Masks}.RGB) {group=FlowingSmoke, maskChannel=Blue, parameterName=FlowingSmoke_TA_MaskChannel, defaultValue=<rgba 0 0 1 0>}
  n18 = Multiply(A=Time(), B=Constant() {r=0.1}) {constB=0.1}
  n19 = Divide(A=Add(A=Sine(None=@n18)))
  NamedRerouteDeclaration(None=Power(Base=Add(A=@n17.R) {constB=0.06}, Exp=LinearInterpolate(Alpha=@n19) {constA=1.4, constB=1.1}) {constExponent=1.2}) {name=SmokeMask, nodeColor=<rgba 1 0 0.0937 1>}
  NamedRerouteDeclaration(None=FunctionInput(Preview=TextureCoordinate()) {bUsePreviewValueAsDefault=true, inputName=In_UV, inputType=FunctionInput_Vector2, sortPriority=50}) {name=In_UV, nodeColor=<rgba 0.7813 0 1 1>}
  n20 = TextureObjectParameter() {
   group=FlowingSmoke
   parameterName=FlowingSmoke_TAMask
   samplerType=SAMPLERTYPE_Masks
   texture=&T_WhiteMask
  }
  NamedRerouteDeclaration(None=FunctionInput(Preview=@n20) {bUsePreviewValueAsDefault=true, inputName=In_TAMask, inputType=FunctionInput_Texture2D, sortPriority=50}) {name=TA_Mask, nodeColor=<rgba 1 0.5391 0 1>}
  n21 = Multiply(
   A=NamedRerouteUsage()
   B=FunctionInput(Preview=ScalarParameter() {defaultValue=3, group=FlowingSmoke, parameterName=FlowingSmoke_BaseSmokeTiling}) {bUsePreviewValueAsDefault=true, inputName=FireSmoke_BaseSmokeTiling, inputType=FunctionInput_Scalar, sortPriority=51}
  ) {constB=5}
  n22 = FunctionInput(Preview=ComponentMask(None=VectorParameter() {group=FlowingSmoke, parameterName=FlowingSmoke_BaseSmokeSpeed, defaultValue=<rgba 0.04 0.2 0 1>}.RGB) {g=true, r=true}) {bUsePreviewValueAsDefault=true, inputName=FireSmoke_BaseSmokeSpeed, inputType=FunctionInput_Vector2, sortPriority=51}
  n23 = TextureSample(UVs=Add(A=NamedRerouteUsage(), B=Panner(Coordinate=@n21, Speed=@n22) {speedX=0.02, speedY=-0.1})) {samplerType=SAMPLERTYPE_Masks, texture=/BRCosmetics/Accessories/FORT_Backpacks/Backpack_F_MED_MistRaven/Textures/T_MistRaven_Backpack_TAMask}
  NamedRerouteDeclaration(None=@n23.B) {name=Base Smoke, nodeColor=<rgba 1 0.6797 0 1>}
  n24 = NamedRerouteUsage()
  n25 = FunctionInput(Preview=ScalarParameter() {defaultValue=3, group=FlowingSmoke, parameterName=FlowingSmoke_NoiseUV}) {bUsePreviewValueAsDefault=true, inputName=FireSmoke_NoiseUV, inputType=FunctionInput_Scalar, sortPriority=53}
  n26 = ComponentMask(None=VectorParameter() {group=FlowingSmoke, parameterName=FlowingSmoke_NoiseSpeed, defaultValue=<rgba 0.04 0.2 0 1>}.RGB) {g=true, r=true}
  n27 = TextureSample(UVs=Panner(
    Coordinate=Multiply(A=@n24, B=@n25) {constB=15}
    Speed=FunctionInput(Preview=@n26) {bUsePreviewValueAsDefault=true, inputName=FireSmoke_NoiseSpeed, inputType=FunctionInput_Vector2, sortPriority=53}
   ) {speedY=0.4}) {
   samplerSource=SSM_Wrap_WorldGroupSettings
   samplerType=SAMPLERTYPE_Grayscale
   texture=/Game/Global/Textures/Default/Noise/LowResBlurredNoiseNormalized
  }
  NamedRerouteDeclaration(None=@n27.R) {name=NoiseUv, nodeColor=<rgba 0.3594 0 1 1>}
  NamedRerouteUsage()
 ]}
