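# Material function that wraps MF_FlowingSmoke for the VinderTech RampCanal glider materials, outputting an animated flowing smoke mask.
# It passes a mask texture and UV input into the inner function through named reroutes, with preview defaults for both.
# It exposes base smoke, swirl, noise and detail smoke tiling, speed and intensity as named parameters in a FlowingSmoke group, so materials can tune them without rewiring.
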MaterialFunction /BRCosmetics/Gadgets/Assets/VinderTech_GliderChute/Glider_RampCanal/Materials/MFW_FlowingSmoke
graph {nodes [
  n1 = FunctionInput(Preview=TextureObject() {samplerType=SAMPLERTYPE_Masks, texture=&T_WhiteMask}) {bUsePreviewValueAsDefault=true, inputName=In_FireSmoke_TAMask, inputType=FunctionInput_Texture2D, sortPriority=50}
  n2 = FunctionInput(Preview=TextureCoordinate()) {bUsePreviewValueAsDefault=true, inputName=In_FireSmoke_UV, inputType=FunctionInput_Vector2, sortPriority=50}
  n3 = VectorParameter() {group=FlowingSmoke, parameterName=FlowingSmoke_BaseSmokeSpeed, defaultValue=<rgba 0.04 0.2 0 1>}
  FunctionOutput(None=MaterialFunctionCall!(
    In_TAMask=NamedRerouteDeclaration(None=@n1) {name=TA_Mask, nodeColor=<rgba 1 0.5391 0 1>}
    In_UV=NamedRerouteDeclaration(None=@n2) {name=In_UV, nodeColor=<rgba 0.7812 0 1 1>}
    FireSmoke_BaseSmokeSpeed=ComponentMask(None=@n3.RGB) {g=true, r=true}
    FireSmoke_BaseSmokeTiling=ScalarParameter() {defaultValue=3, group=FlowingSmoke, parameterName=FlowingSmoke_BaseSmokeTiling}
    FireSmoke_SwirlUVTiling=ScalarParameter() {defaultValue=3, group=FlowingSmoke, parameterName=FlowingSmoke_SwirlUVTiling}
    FireSmoke_SwirlUVSpeed=ScalarParameter() {defaultValue=0.1, group=FlowingSmoke, parameterName=FlowingSmoke_SwirlUVSpeed}
    FireSmoke_SwirlIntensity=ScalarParameter() {defaultValue=0.4, group=FlowingSmoke, parameterName=FlowingSmoke_SwirlIntensity}
    FireSmoke_NoiseUV=ScalarParameter() {defaultValue=3, group=FlowingSmoke, parameterName=FlowingSmoke_NoiseUV}
    FireSmoke_NoiseSpeed=VectorParameter() {group=FlowingSmoke, parameterName=FlowingSmoke_NoiseSpeed, defaultValue=<rgba 0.04 0.2 0 1>}.RGB
    FireSmoke_NoiseIntensity=ScalarParameter() {defaultValue=0.02, group=FlowingSmoke, parameterName=FlowingSmoke_NoiseIntensity}
    FireSmoke_DetailSmoketiling=ScalarParameter() {defaultValue=3, group=FlowingSmoke, parameterName=FlowingSmoke_DetailSmoketiling}
    FireSmoke_DetailSmokeSpeed=VectorParameter() {group=FlowingSmoke, parameterName=FlowingSmoke_DetailSmokeSpeed, defaultValue=<rgba -0.04 0.4 0 1>}.RGB
   ) {materialFunction=&/MF_FlowingSmoke}.Out_FireSmokeMask)
 ]}
