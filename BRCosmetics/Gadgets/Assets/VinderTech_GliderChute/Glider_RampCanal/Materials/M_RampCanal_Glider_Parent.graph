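# Parent material for the RampCanal glider, a skeletal-mesh cloth-shaded surface built on the 2023 anime shader's Pre, Calculations and Post functions.
# Eye animation adjusts lit and shaded colors, which are then blended toward scaled versions before the Post function.
# Custom emissive layers combine Fresnel, panning-noise smoke, flowing wing smoke, a sphere-masked throat glow in pre-skinning local space and color offset.
# A packed texture supplies emissive mask and detail channels.
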Material /BRCosmetics/Gadgets/Assets/VinderTech_GliderChute/Glider_RampCanal/Materials/M_RampCanal_Glider_Parent
graph {
 properties {
  bAutomaticallySetUsageInEditor=false
  bUseMaterialAttributes=true
  bUsedWithMeshDeformer=true
  bUsedWithMorphTargets=true
  bUsedWithSkeletalMesh=true
  pixelDepthOffsetMode=PDOM_Legacy
  shadingModel=MSM_Cloth
  usedShadingModels=Cloth
 }
 nodes [
  n1 = MaterialFunctionCall!(CustomUV) {materialFunction=/Game/Global/Materials/Character_Parent_2023/Anime_Parent_2023/MaterialFunctions/MFC_AnimeShader_2023_Pre}
  n2 = MaterialFunctionCall!(CustomLightVector, OverrideNormal_WS, STT=@n1.STT, Normal_Map=@n1.STT, DFL=@n1.STT, FX=@n1.STT) {materialFunction=/Game/Global/Materials/Character_Parent_2023/Anime_Parent_2023/MaterialFunctions/MFC_AnimeShader_2023_Calculations}
  n3 = VertexColor()
  n4 = TextureSampleParameter2D() {
   group=01_EyeAnimation_Parameters
   parameterName=EyeMask
   samplerSource=SSM_Wrap_WorldGroupSettings
   samplerType=SAMPLERTYPE_Masks
   sortPriority=-32
   texture=&T_WhiteMask
  }
  n5 = StaticSwitchParameter(True=AppendVector(A=@n3, B=@n3), False=@n4.RGBA) {group=01_EyeAnimation_Parameters, parameterName=UseVertexColorForEyeMask, sortPriority=-33}
  n6 = TextureCoordinate() {coordinateIndex=1}
  n7 = Sign(None=Subtract(A=ComponentMask(None=@n6) {r=true}) {constB=0.0001})
  n8 = MaterialFunctionCall!(
   Color Lit=NamedRerouteDeclaration(None=@n1.Color_Lit) {name=Color_Lit, nodeColor=<rgba 0.6641 1 0 1>}
   Color Shaded=@n1.Color_Shaded
   Use UV2=StaticBoolParameter() {defaultValue=true, group=01_EyeAnimation_Parameters, parameterName=Use UV2 for Eye Animation, sortPriority=-30}
   Mask=StaticSwitchParameter(
    True=ChannelMaskParameter(None=@n5) {group=01_EyeAnimation_Parameters, parameterName=EyeMaskChannel, sortPriority=-32}.R
    False=Saturate(None=@n7)
   ) {group=01_EyeAnimation_Parameters, parameterName=UseEyeMask, sortPriority=-33}
  ) {materialFunction=/Game/Global/Materials/Character_Parent_2023/Anime_Parent_2023/MaterialFunctions/MF_AnimeEyeAnimation_v2}
  n9 = ScalarParameter() {defaultValue=0.3, group=RampCanal, parameterName=BaseColor_blend}
  MaterialFunctionCall!(
   Shading_Gradient=@n2.Shading_Gradient
   Threshold_Gradient=@n2.Shading_Gradient
   Specular_Gradient=@n2.Shading_Gradient
   Sheen_Mask=@n2.Shading_Gradient
   Thickness=@n2.Shading_Gradient
   DFL_Mask=@n2.Shading_Gradient
   FX=@n2.Shading_Gradient
   Metallic_Gradient=@n2.Shading_Gradient
   RimLight_Gradient=@n2.Shading_Gradient
   NormalBuffer=@n2.Shading_Gradient
   --------
   CINE_LightSources=@n2.Shading_Gradient
   CINE_LightColor=@n2.Shading_Gradient
   ---------
   Color_Lit=LinearInterpolate(A=@n8.Lit, B=Multiply(A=@n8.Lit, B=@n9) {constB=0.2}, Alpha=NamedRerouteUsage()) {constB=0}
   Color_Shaded=LinearInterpolate(A=@n8.Shaded, B=Multiply(A=@n8.Shaded, B=@n9) {constB=0.2}, Alpha=NamedRerouteUsage()) {constB=0}
   SRM=@n1.SRM
   Emissive=NamedRerouteUsage()
   DFL_Color=NamedRerouteDeclaration(None=@n1.DFL_Color) {name=DFL_Color, nodeColor=<rgba 1 0 0.9844 1>}
   Specular_Color=@n1.SRM
   RimLight_Color=@n1.SRM
  ) {materialFunction=/Game/Global/Materials/Character_Parent_2023/Anime_Parent_2023/MaterialFunctions/MFC_AnimeShader_2023_Post}
  n10 = Constant() {r=1}
  n11 = LinearInterpolate(
   A=ScalarParameter() {defaultValue=0.9, group=RampCanal, parameterName=Fresnel_Emissive_Min}
   B=ScalarParameter() {defaultValue=0.4, group=RampCanal, parameterName=Fresnel_Emissive_Max}
   Alpha=Saturate(None=Fresnel(ExponentIn=@n10) {exponent=1})
  )
  n12 = Panner(Coordinate=TextureCoordinate()) {speedX=0.07, speedY=0.07}
  n13 = Power(Base=Multiply(A=TextureSample(UVs=@n12) {samplerType=SAMPLERTYPE_Grayscale, texture=/Game/Global/Textures/Default/Noise/LowResBlurredNoiseNormalized}.R) {constB=1.5})
  n14 = VertexInterpolator(VS=LocalPosition() {localOrigin=InstancePreSkinning}.XYZ)
  n15 = Constant4Vector() {constant=<rgba 0 -194.452 -15.5226 1>}
  n16 = Constant() {r=49.8573}
  n17 = Constant()
  n18 = LinearInterpolate(
   A=ScalarParameter() {defaultValue=0.75, group=RampCanal, parameterName=Noise_Emissive_Min}
   B=ScalarParameter() {defaultValue=6, group=RampCanal, parameterName=Noise_Emissive_Max}
   Alpha=LinearInterpolate(A=Saturate(None=@n13), B=Constant(), Alpha=SphereMask(A=@n14.PS, B=@n15.RGB, Radius=@n16, Hardness=@n17)) {constB=0}
  ) {constA=1, constB=2}
  n19 = ScalarParameter() {defaultValue=-0.1, group=RampCanal, parameterName=EmissiveDetail_MIn}
  n20 = ScalarParameter() {defaultValue=3, group=RampCanal, parameterName=EmissiveDetail_Max}
  n21 = NamedRerouteUsage()
  NamedRerouteDeclaration(None=Multiply(A=Multiply(A=@n11, B=@n18), B=LinearInterpolate(A=@n19, B=@n20, Alpha=@n21) {constA=1, constB=2})) {name=RampCanalBaseSmoke, nodeColor=<rgba 0 0.9062 1 1>}
  NamedRerouteDeclaration(None=@n1.Emissive) {name=In_Emissive, nodeColor=<rgba 1 0.7734 0 1>}
  n22 = VertexInterpolator(VS=LocalPosition() {localOrigin=InstancePreSkinning}.XYZ)
  n23 = Constant4Vector() {constant=<rgba 0 -194.452 -15.5226 1>}
  n24 = Constant() {r=19.8115}
  n25 = Constant() {r=0.137}
  NamedRerouteDeclaration(None=LinearInterpolate(
    B=ScalarParameter() {defaultValue=2, group=RampCanal, parameterName=ThroatGlow}
    Alpha=SphereMask(A=@n22.PS, B=@n23.RGB, Radius=@n24, Hardness=@n25)
   ) {constA=1, constB=3}) {name=ThroatGlow, nodeColor=<rgba 0.4531 0 1 1>}
  n26 = NamedRerouteUsage()
  NamedRerouteDeclaration(None=LinearInterpolate(
    A=ScalarParameter() {defaultValue=0.6, group=RampCanal, parameterName=WingEmissive_Min}
    B=ScalarParameter() {defaultValue=6, group=RampCanal, parameterName=WingEmissive_Max}
    Alpha=MaterialFunctionCall!(In_FireSmoke_TAMask=@n26, In_FireSmoke_UV) {materialFunction=&/MFW_FlowingSmoke}.Result
   ) {constA=1, constB=2}) {name=WingSmoke, nodeColor=<rgba 0 1 0.1328 1>}
  n27 = LinearInterpolate(
   A=VectorParameter() {group=RampCanal, parameterName=EmissiveColorOffset, defaultValue=<rgba 0.011 0.0908 0.8963 1>}.RGB
   B=NamedRerouteUsage()
   Alpha=ScalarParameter() {defaultValue=0.7, group=RampCanal, parameterName=EmissiveColorOffset_Blend}
  ) {constAlpha=0.8}
  n28 = NamedRerouteUsage()
  n29 = NamedRerouteUsage()
  n30 = LinearInterpolate(
   A=NamedRerouteUsage()
   B=Multiply(A=Multiply(A=@n27, B=@n28), B=LinearInterpolate(Alpha=@n29) {constA=1, constB=0.05})
   Alpha=NamedRerouteUsage()
  )
  n31 = NamedRerouteUsage()
  NamedRerouteDeclaration(None=Multiply(A=Multiply(A=@n30, B=@n31), B=NamedRerouteUsage())) {name=Out_Emissive, nodeColor=<rgba 1 0 0.1875 1>}
  n32 = TextureObject() {samplerType=SAMPLERTYPE_Masks, texture=/BRCosmetics/Gadgets/Assets/VinderTech_GliderChute/Glider_RampCanal/Textures/T_RampCanal_Glider_E}
  n33 = TextureSample(UVs=TextureCoordinate(), Tex=NamedRerouteDeclaration(None=@n32) {name=TA_Mask, nodeColor=<rgba 0.1719 1 0 1>}) {samplerSource=SSM_Wrap_WorldGroupSettings}
  NamedRerouteDeclaration(None=@n33.R) {name=EmissiveMask, nodeColor=<rgba 1 0 0.8906 1>}
  NamedRerouteDeclaration(None=@n33.G) {name=EmissiveDetail, nodeColor=<rgba 1 0.4688 0 1>}
 ]
}
